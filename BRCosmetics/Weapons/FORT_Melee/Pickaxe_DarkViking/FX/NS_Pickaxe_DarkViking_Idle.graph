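# Idle Niagara effect for the Dark Viking pickaxe, producing blue ice flames, embers and glow at weapon sockets read through a Socket Reader.
# Flames and embers spawn continuously, with embers also spawning by distance moved and stopping while User.bisSwinging is true.
# Burst-spawned glow meshes, sprites and a light fade in with system age, timed separately for gameplay and front end.
# Each emitter packs several layers using spawn-group VisibilityTags and MergeHelper switches.
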NiagaraSystem /BRCosmetics/Weapons/FORT_Melee/Pickaxe_DarkViking/FX/NS_Pickaxe_DarkViking_Idle
dependencies {
 usedDataInterfaces [
  &NiagaraDataInterfaceCurve
  &NiagaraDataInterfaceColorCurve
  &NiagaraDataInterfaceSpriteRendererInfo
  &NiagaraDataInterfaceVectorCurve
  &NiagaraDataInterfaceSocketReader
 ]
 usedDynamicInputs []
 usedModules [
  &self:SystemSpawnScript.SystemScriptSource.SystemScriptGraph.NiagaraNodeAssignment_0.SetVariables_#1
  &SystemState
 ]
}
emitters {
 Embers_Packed {
  emitterUpdateScript [
   &EmitterState(Inactive Response=::NewEnumerator1, Life Cycle Mode=::NewEnumerator1, Loop Duration=1, UseLoopDelay=false)
   &SpawnPerUnit(
    Module.Use Max Movement Threshold=true
    Module.Use Movement Tolerance=false
    Spawn Probability:NiagaraFloat=&MakeCustomFloatFromBool(Bool=$User.bisSwinging, True Float=0, False Float=1)
    Spawn Spacing=250
    Use Spawn Probability=true
   ) {module=SpawnPerUnit001}
   &SpawnRate(
    Spawn Probability=&MakeCustomFloatFromBool(Bool=$User.bisSwinging, True Float=0, False Float=1)
    SpawnGroup=1
    SpawnRate=5
    Use Spawn Probability=true
   ) {module=SpawnRate002}
   &SpawnPerUnit(
    Module.Use Max Movement Threshold=true
    Module.Use Movement Tolerance=false
    Spawn Group=1
    Spawn Probability:NiagaraFloat=&MakeCustomFloatFromBool(Bool=$User.bisSwinging, True Float=0, False Float=1)
    Spawn Spacing=250
    Use Spawn Probability=true
   )
  ]
  particleSpawnScript [
   /SetVariables(Particles.VisibilityTag=$Emitter.SpawnGroup) {isSetParameters=true, module=SetVariables_#2}
   &InitializeParticle(
    Color=&MergeHelper_ColorSwitch(
     0=&RandomRangeLinearColor(Minimum=<rgba 0 0.5263 1 1>, Maximum=<rgba 0 0.2415 0.8 1>)
     1=<rgba 3 3 3 1>
     2=?
     3=?
     Index=$Particles.VisibilityTag
     Num=::NewEnumerator0
    )
    Lifetime=&MergeHelper_FloatSwitch(
     0=1
     1=&RandomRangeFloat(Minimum=0.5, Maximum=2, Evaluation Type=::NewEnumerator1)
     2=&RandomRangeFloat(Minimum=0.5, Maximum=2, Evaluation Type=::NewEnumerator1)
     3=?
     Index=$Particles.VisibilityTag
     Num=::NewEnumerator0
    )
    Sprite Rotation Mode=::NewEnumerator2
    Sprite Size=&MergeHelper_Vector2Switch(
     0=&Vector2DFromFloat(Value=&RandomRangeFloat(Minimum=1.8, Maximum=3, Evaluation Type=::NewEnumerator1))
     1=&RandomRangeVector2D(Minimum=<0.5 0.8>, Maximum=<1 1>, Evaluation Type=::NewEnumerator1)
     2=&RandomRangeVector2D(Minimum=<0.2 0.2>, Maximum=<2 2>, Evaluation Type=::NewEnumerator1)
     3=?
     4=?
     Index=$Particles.VisibilityTag
     Num=::NewEnumerator0
    )
    Sprite Size Mode=::NewEnumerator4
    Sprite UV Mode=::NewEnumerator1
    UsePositionOffset=false
    Write Material Random=true
   )
   &SocketLocation(
    Interpolate Socket Sampling=true
    Orientation Sampling=::NewEnumerator1
    SamplePosOffset=true
    Sampled Position Offset=<0 0 5>
    Socket Index=1
    Socket Reader=$System.Socket Reader
    Socket Sampling Mode=::NewEnumerator3
    VelocitySpeedLimit=false
   )
   &ShapeLocation(MaskSpawnGroup=false, Override Seed=false, Sphere Radius=2, Sphere Surface Distribution=0.5, UseOwnerScale=true)
   &AddVelocity(Velocity=&MergeHelper_VectorSwitch(
     0=<0 0 0>
     1=&RandomRangeVector(Minimum=<0 0 25>, Maximum=<0 0 50>, Evaluation Type=::NewEnumerator1)
     2=&RandomRangeVector(Minimum=<0 0 25>, Maximum=<0 0 50>, Evaluation Type=::NewEnumerator1)
     3=?
     Index=$Particles.VisibilityTag
     Num=::NewEnumerator0
    ))
   &AddVelocity(
    Module.Use Falloff Distance=false
    Override Seed=false
    Velocity Mode=::NewEnumerator1
    Velocity Origin=$Output.SocketLocation.SampledPosition
    Velocity Speed=&MergeHelper_FloatSwitch(0=3, 1=20, 2=?, 3=?, Index=$Particles.VisibilityTag, Num=::NewEnumerator0)
   ) {module=AddVelocity001}
  ]
  particleUpdateScript [
   &ParticleState()
   &Drag(Drag=3, Use Linear Drag=true, Use Rotational Drag=true)
   &ScaleSpriteSize(
    Scale Factor=&Vector2DFromFloat(Value=&MergeHelper_FloatSwitch(
      0=&FloatFromCurve(FloatCurve=DataInterface() {propertyValues {Curve {keys [{arriveTangent=6.5112, interpMode=RCIM_Cubic, leaveTangent=6.5112, tangentMode=RCTM_User, time=0, value=0}, <at 0.4 1>]}}})
      1=&FloatFromCurve(FloatCurve=DataInterface() {propertyValues {Curve {keys [
           {arriveTangent=11.6538, interpMode=RCIM_Cubic, leaveTangent=11.6538, tangentMode=RCTM_User, time=0, value=0.0211}
           {arriveTangent=0.5535, interpMode=RCIM_Cubic, leaveTangent=0.5535, tangentMode=RCTM_User, time=0.1836, value=0.9874}
           <at 1 0.5 RCIM_Cubic>
          ]}}})
      2=?
      3=?
      Index=$Particles.VisibilityTag
      Num=::NewEnumerator0
     ))
    Scale Sprite Size Mode=::NewEnumerator2
   )
   &ScaleColor(
    Scale Alpha=&FloatFromCurve(FloatCurve=DataInterface() {propertyValues {Curve {keys [<at 0.75 1>,<at 1 0>]}}})
    Scale RGB=&VectorFromFloat(Value=&FloatFromCurve(FloatCurve=DataInterface() {propertyValues {Curve {keys [<at 0.75 1>,<at 1 0>]}}}))
    ScaleA=true
    ScaleRGB=true
   )
   &AccelerationForce(Acceleration=&MergeHelper_VectorSwitch(0=<0 0 -25>, 1=<0 0 100>, 2=?, 3=?, Index=$Particles.VisibilityTag, Num=::NewEnumerator0))
   &SolveForcesAndVelocity(Clamp Velocity=false, Limit Acceleration=false)
   &ScaleSpriteSizeBySpeed(
    Max Scale Factor=&MergeHelper_Vector2Switch(0=<1 3>, 1=<1 10>, 2=?, 3=?, 4=?, Index=$Particles.VisibilityTag, Num=::NewEnumerator0)
    Velocity Threshold=600
   )
   &FN_Sprite_Attributes(
    Color=&MakeLinearColorFromVectorAndFloat(
     "Vector (RGB)"=&MakeVectorFromLinearColorRGB(Color=&ColorFromCurve(ColorCurve=DataInterface() {propertyValues {
         AlphaCurve {keys [<at 0 1 RCIM_Cubic>]}
         BlueCurve {keys [<at 0 1.5 RCIM_Cubic>,<at 1 0.7 RCIM_Cubic>]}
         GreenCurve {keys [<at 0 0.4529 RCIM_Cubic>,<at 1 0.3679 RCIM_Cubic>]}
         RedCurve {keys [<at 0 0 RCIM_Cubic>,<at 1 0 RCIM_Cubic>]}
        }}))
     "Float (Alpha)"=&FloatFromCurve(FloatCurve=DataInterface() {propertyValues {Curve {keys [
          {arriveTangent=11.7282, interpMode=RCIM_Cubic, leaveTangent=11.7282, tangentMode=RCTM_User, time=0, value=0}
          <at 0.2 1 RCIM_Cubic>
          <at 1 0>
         ]}}})
    )
    Create Color Attribute=true
    Module Output:ENiagara_AttributeSamplingApplyOutput=::NewEnumerator1
   )
  ]
  rendererClasses [&NiagaraSpriteRendererProperties]
  renderers [
   {rendererClass=&NiagaraSpriteRendererProperties}
   {rendererClass=&NiagaraSpriteRendererProperties, rendererIndex=1}
  ]
 }
 Flames_Packed {
  emitterUpdateScript [
   &EmitterState(Inactive Response=::NewEnumerator1, Life Cycle Mode=::NewEnumerator1, Loop Duration=1, UseLoopDelay=false)
   &SpawnRate(SpawnRate=10, Use Spawn Probability=false)
   &SpawnRate(SpawnGroup=1, SpawnRate=10, Use Spawn Probability=false) {module=SpawnRate001}
   &SpawnRate(SpawnGroup=2, SpawnRate=3, Use Spawn Probability=false) {module=SpawnRate002}
  ]
  particleSpawnScript [
   /SetVariables(Particles.VisibilityTag=$Emitter.SpawnGroup) {isSetParameters=true, module=SetVariables_#2}
   &InitializeParticle(
    Color=&MergeHelper_ColorSwitch(0=<rgba 5 5 5 1>, 1=<rgba 1 1 1 1>, 2=?, 3=?, Index=$Particles.VisibilityTag, Num=::NewEnumerator0)
    Lifetime=&MergeHelper_FloatSwitch(
     0=&RandomRangeFloat(Minimum=0.5, Maximum=1, Evaluation Type=::NewEnumerator1)
     1=&RandomRangeFloat(Minimum=0.5, Maximum=0.8, Evaluation Type=::NewEnumerator1)
     2=1
     3=?
     Index=$Particles.VisibilityTag
     Num=::NewEnumerator1
    )
    Mesh Scale=<0.3 0.3 0.2>
    Mesh Scale Mode=::NewEnumerator4
    Sprite Rotation Mode=::NewEnumerator2
    Sprite Size=&MergeHelper_Vector2Switch(
     0=<0 0>
     1=&Vector2DFromFloat(Value=&RandomRangeFloat(Minimum=18, Maximum=25))
     2=&Vector2DFromFloat(Value=&RandomRangeFloat(Minimum=1.8, Maximum=3, Evaluation Type=::NewEnumerator1))
     3=?
     4=?
     Index=$Particles.VisibilityTag
     Num=::NewEnumerator1
    )
    Sprite Size Mode=::NewEnumerator4
    Sprite UV Mode=::NewEnumerator1
    UsePositionOffset=false
    Write Material Random=true
   )
   &SocketLocation(
    Orientation Sampling=::NewEnumerator1
    SamplePosOffset=true
    Sampled Position Offset=<0 0 5>
    Socket Index=1
    Socket Reader=$System.Socket Reader
    Socket Sampling Mode=::NewEnumerator3
    VelocitySpeedLimit=false
   )
   &ShapeLocation(
    MaskSpawnGroup=false
    Override Seed=false
    Sphere Radius=&MergeHelper_FloatSwitch(0=0, 1=0, 2=10, 3=?, Index=$Particles.VisibilityTag, Num=::NewEnumerator1)
    UseOwnerScale=true
   )
   &AddVelocity(Velocity=&MergeHelper_VectorSwitch(
     0=<0 0 1>
     1=&RandomRangeVector(Minimum=<0 0 10>, Maximum=<0 0 30>, Evaluation Type=::NewEnumerator1)
     2=<0 0 0>
     3=?
     Index=$Particles.VisibilityTag
     Num=::NewEnumerator1
    ))
   &AddVelocity(
    Module.Use Falloff Distance=false
    Override Seed=false
    Velocity Mode=::NewEnumerator1
    Velocity Origin=$Output.SocketLocation.SampledPosition
    Velocity Speed=&MergeHelper_FloatSwitch(0=0, 1=0, 2=10, 3=?, Index=$Particles.VisibilityTag, Num=::NewEnumerator1)
   ) {module=AddVelocity001}
   &DynamicMaterialParameters(
    Dynamic Parameter Output:ENiagara_AttributeSamplingApplyOutput=::NewEnumerator1
    Index 0 Param 1=&RandomRangeFloat(Minimum=0, Maximum=1, Evaluation Type=::NewEnumerator1)
    Module.Index0Float0Write=true
    Module.Index0Float1Write=false
    Module.Index0Float2Write=false
    Module.Index0Float3Write=false
   )
   &InitialMeshOrientation(Mesh Orientation Mode=::NewEnumerator2, Orientation Vector=<0 0 1>, Use Rotation Vector=false)
  ]
  particleUpdateScript [
   &ParticleState()
   &Drag(
    Drag=&MergeHelper_FloatSwitch(0=0, 1=1, 2=3, 3=?, Index=$Particles.VisibilityTag, Num=::NewEnumerator1)
    Use Linear Drag=true
    Use Rotational Drag=true
   )
   &SocketLocation(
    Orientation Sampling=::NewEnumerator1
    Position Sampling=::NewEnumerator1
    SamplePosOffset=true
    Sampled Position Offset=<0 0 5>
    Socket Index=1
    Socket Reader=$System.Socket Reader
    Socket Sampling Mode=::NewEnumerator3
    VelocitySpeedLimit=false
   ) {module=SocketLocation001}
   /SetVariables(Particles.Position=/Niagara/DynamicInputs/Helpers/MaskPosition(
     Mask Source:NiagaraInt32=1
     Mask Parameter:NiagaraInt32=$Particles.VisibilityTag
     Mask:NiagaraInt32=0
     Masked Position:NiagaraPosition=$Output.SocketLocation001.SampledPosition
     Passthrough Position:NiagaraPosition=$Particles.Position
    )) {isSetParameters=true}
   &SubUVAnimation(
    Output Alternate Renderer Binding=true
    "Sprite Renderer "=DataInterface() {propertyValues {SpriteRenderer=&self:Flames_Packed_0.NiagaraSpriteRendererProperties_1}}
    UseEndFrame=false
    UseStartFrame=false
   )
   &ScaleSpriteSize(
    Scale Factor=&Vector2DFromFloat(Value=&MergeHelper_FloatSwitch(
      0=0
      1=&FloatFromCurve(FloatCurve=DataInterface() {propertyValues {Curve {keys [<at 0 0.0211 RCIM_Cubic>, {interpMode=RCIM_Cubic, tangentMode=RCTM_User, time=0.1836, value=1}, <at 1 0.5 RCIM_Cubic>]}}})
      2=&FloatFromCurve(FloatCurve=DataInterface() {propertyValues {Curve {keys [{arriveTangent=6.5112, interpMode=RCIM_Cubic, leaveTangent=6.5112, tangentMode=RCTM_User, time=0, value=0}, <at 0.4 1>]}}})
      3=?
      Index=$Particles.VisibilityTag
      Num=::NewEnumerator1
     ))
    Scale Sprite Size Mode=::NewEnumerator2
   )
   &ScaleMeshSize(Scale Factor=&VectorFromCurve(VectorCurve=DataInterface() {propertyValues {
       XCurve {keys [<at 0 0.75 RCIM_Cubic>, {arriveTangent=0.25, interpMode=RCIM_Cubic, leaveTangent=0.25, time=0.5, value=1}, <at 1 1 RCIM_Cubic>]}
       YCurve {keys [<at 0 1 RCIM_Cubic>,<at 0.5 1 RCIM_Cubic>,<at 1 1 RCIM_Cubic>]}
       ZCurve {keys [<at 0 1 RCIM_Cubic>,<at 0.5 1 RCIM_Cubic>,<at 1 1 RCIM_Cubic>]}
      }}))
   &ScaleColor(
    Scale Alpha=&FloatFromCurve(FloatCurve=DataInterface() {propertyValues {Curve {keys [
         {arriveTangent=11.7282, interpMode=RCIM_Cubic, leaveTangent=11.7282, tangentMode=RCTM_User, time=0, value=0}
         <at 0.2 1 RCIM_Cubic>
         <at 1 0>
        ]}}})
    Scale RGB=&MergeHelper_VectorSwitch(
     0=<1 1 1>
     1=&VectorFromCurve(VectorCurve=DataInterface() {propertyValues {XCurve {keys [<at 0 2>,<at 1 0.5>]}, YCurve {keys [<at 0 2>,<at 1 0.5>]}, ZCurve {keys [<at 0 2>,<at 1 0.5>]}}})
     2=?
     3=?
     Index=$Particles.VisibilityTag
     Num=::NewEnumerator0
    )
    ScaleA=true
    ScaleRGB=true
   )
   &AccelerationForce(Acceleration=&MergeHelper_VectorSwitch(0=<0 0 0>, 1=<0 0 100>, 2=<0 0 -25>, 3=?, Index=$Particles.VisibilityTag, Num=::NewEnumerator1))
   &SolveForcesAndVelocity(Clamp Velocity=false, Limit Acceleration=false)
   &DynamicMaterialParameters(
    Dynamic Parameter Output:ENiagara_AttributeSamplingApplyOutput=::NewEnumerator1
    Index 0 Param 1=&FloatFromCurve(FloatCurve=DataInterface() {propertyValues {Curve {keys [<at 0.5 0 RCIM_Cubic>, {arriveTangent=6.5632, interpMode=RCIM_Cubic, leaveTangent=6.5632, tangentMode=RCTM_User, time=1, value=1}]}}})
    Index 0 Param 2=1
    Index 0 Param 3=1
    Index 0 Param 4=&FloatFromCurve(FloatCurve=DataInterface() {propertyValues {Curve {keys [<at 0 0>,<at 1 10>]}}})
    Module.Index0Float0Write=true
    Module.Index0Float1Write=true
    Module.Index0Float2Write=true
    Module.Index0Float3Write=true
   ) {module=DynamicMaterialParameters001}
   &FN_Sprite_Attributes(
    Color=&MakeLinearColorFromVectorAndFloat(
     "Vector (RGB)"=&MakeVectorFromLinearColorRGB(Color=&Multiply_LinearColorByFloat(
       Linear Color=&RandomRangeLinearColor(Minimum=<rgba 0 0.5263 1 1>, Maximum=<rgba 0 0.2415 0.8 1>)
       Float=&FloatFromCurve(FloatCurve=DataInterface() {propertyValues {Curve {keys [<at 0.75 1>,<at 1 0>]}}})
      ))
     "Float (Alpha)"=&FloatFromCurve(FloatCurve=DataInterface() {propertyValues {Curve {keys [<at 0.75 1>,<at 1 0>]}}})
    )
    Create Color Attribute=true
    Module Output:ENiagara_AttributeSamplingApplyOutput=::NewEnumerator1
   )
  ]
  rendererClasses [&NiagaraMeshRendererProperties, &NiagaraSpriteRendererProperties]
  renderers [
   {rendererClass=&NiagaraMeshRendererProperties, rendererIndex=0}
   {rendererClass=&NiagaraSpriteRendererProperties, rendererIndex=1}
   {rendererClass=&NiagaraSpriteRendererProperties, rendererIndex=2}
  ]
 }
 IceFlameGlow_Packed {
  emitterUpdateScript [
   &EmitterState(
    Inactive Response=::NewEnumerator1
    Life Cycle Mode=::NewEnumerator1
    Loop Behavior=::NewEnumerator1
    Loop Duration Mode=::NewEnumerator1
    UseLoopDelay=false
   )
   &SpawnBurst_Instantaneous(Use Spawn Probability=false, UseLoopCountLimit=false)
   &SpawnBurst_Instantaneous(Spawn Group=1, Use Spawn Probability=false, UseLoopCountLimit=false) {module=SpawnBurst_Instantaneous001}
   &SpawnBurst_Instantaneous(Spawn Group=2, Use Spawn Probability=false, UseLoopCountLimit=false) {module=SpawnBurst_Instantaneous002}
   &SpawnBurst_Instantaneous(Spawn Group=3, Use Spawn Probability=false, UseLoopCountLimit=false) {module=SpawnBurst_Instantaneous003}
  ]
  particleSpawnScript [
   /SetVariables(Particles.VisibilityTag=$Emitter.SpawnGroup) {isSetParameters=true}
   &InitializeParticle(
    Color=&MergeHelper_ColorSwitch(0=<rgba 15 15 15 1>, 1=<rgba 1 1 1 1>, 2=<rgba 0 4.5225 15 1>, 3=<rgba 1 1 1 1>, Index=$Particles.VisibilityTag, Num=::NewEnumerator2)
    Mesh Scale Mode=::NewEnumerator4
    Sprite Size Mode=::NewEnumerator3
    Uniform Sprite Size=25
    UsePositionOffset=false
    Write Material Random=true
   )
   &SocketLocation(
    Orientation Sampling=::NewEnumerator1
    SamplePosOffset=true
    Sampled Position Offset=&MergeHelper_VectorSwitch(0=<0 0 0>, 1=<0 0 0>, 2=<0 0 3>, 3=<0 0 5>, Index=$Particles.VisibilityTag, Num=::NewEnumerator2)
    Socket Index=&MergeHelper_IntSwitch(0=0, 1=1, 2=1, 3=1, 4=?, 5=?, Index=$Particles.VisibilityTag, Num=::NewEnumerator2)
    Socket Reader=$System.Socket Reader
    Socket Sampling Mode=::NewEnumerator3
    VelocitySpeedLimit=false
   )
   &DynamicMaterialParameters(
    Index 0 Param 1=&RandomRangeFloat(Minimum=0, Maximum=1, Evaluation Type=::NewEnumerator1)
    Module.Index0Float0Write=true
    Module.Index0Float1Write=false
    Module.Index0Float2Write=false
    Module.Index0Float3Write=false
   )
   &CameraOffset(Camera Offset Amount=&MergeHelper_FloatSwitch(0=0, 1=25, 2=10, 3=0, Index=$Particles.VisibilityTag, Num=::NewEnumerator2))
  ]
  particleUpdateScript [
   &ParticleState(Kill Particles When Lifetime Has Elapsed=false)
   &ScaleColor(
    Scale Alpha=&FN_FrontEndFloatModifier(
     Gameplay=&FloatFromCurve(FloatCurve=DataInterface() {propertyValues {Curve {keys [<at 0 0 RCIM_Cubic>,<at 0.8 1 RCIM_Cubic>]}}}, CurveIndex=$Engine.System.Age)
     Front End=&FloatFromCurve(FloatCurve=DataInterface() {propertyValues {Curve {keys [<at 0.35 0 RCIM_Cubic>,<at 1.2 1 RCIM_Cubic>]}}}, CurveIndex=$Engine.System.Age)
    )
    ScaleA=true
    ScaleRGB=true
   )
   &DynamicMaterialParameters(
    Dynamic Parameter Output:ENiagara_AttributeSamplingApplyOutput=::NewEnumerator1
    Index 0 Param 1=1
    Index 0 Param 2=0.5
    Index 0 Param 3=0.7571
    Index 0 Param 4=1
    Module.Index0Float0Write=true
    Module.Index0Float1Write=true
    Module.Index0Float2Write=true
    Module.Index0Float3Write=true
   ) {module=DynamicMaterialParameters001}
   &DynamicMaterialParameters(
    Dynamic Parameter Output:ENiagara_AttributeSamplingApplyOutput=::NewEnumerator1
    Index 0 Param 1=0.25
    Index 0 Param 2=2
    Index 0 Param 3=1
    Index 0 Param 4=1
    Module.Index0Float0Write=true
    Module.Index0Float1Write=true
    Module.Index0Float2Write=true
    Module.Index0Float3Write=true
   ) {module=DynamicMaterialParameters002}
   &FN_Light_Attributes(
    Color=&Multiply_LinearColorByFloat(
     Linear Color=&ColorFromCurve(ColorCurve=DataInterface() {propertyValues {
        AlphaCurve {keys [<at 0 1 RCIM_Cubic>]}
        BlueCurve {keys [<at 0 0.723 RCIM_Cubic>,<at 1 0.3515 RCIM_Cubic>]}
        GreenCurve {keys [<at 0 0.218 RCIM_Cubic>,<at 1 0.185 RCIM_Cubic>]}
        RedCurve {keys [<at 0 0 RCIM_Cubic>,<at 1 0 RCIM_Cubic>]}
       }})
     Float=10
    )
    Create Color Attribute=true
    Create Exponent Attribute=true
    Create Radius Attribute=true
    Exponent=8
    Radius=48
    Scale Color=&FN_FrontEndFloatModifier(
     Gameplay=&FloatFromCurve(FloatCurve=DataInterface() {propertyValues {Curve {keys [<at 0 0 RCIM_Cubic>,<at 0.8 1 RCIM_Cubic>]}}}, CurveIndex=$Engine.System.Age)
     Front End=&FloatFromCurve(FloatCurve=DataInterface() {propertyValues {Curve {keys [<at 0.35 0 RCIM_Cubic>,<at 1.2 1 RCIM_Cubic>]}}}, CurveIndex=$Engine.System.Age)
    )
   )
  ]
  rendererClasses [
   &NiagaraMeshRendererProperties
   &NiagaraSpriteRendererProperties
   &NiagaraLightRendererProperties
  ]
  renderers [
   {rendererClass=&NiagaraMeshRendererProperties, rendererIndex=0}
   {rendererClass=&NiagaraSpriteRendererProperties, rendererIndex=1}
   {rendererClass=&NiagaraSpriteRendererProperties, rendererIndex=2}
   {rendererClass=&NiagaraLightRendererProperties, rendererIndex=3}
  ]
 }
}
summary {
 systemName=NS_Pickaxe_DarkViking_Idle
 emitters [
  {
   emitterName=IceFlameGlow_Packed
   rendererClasses [
    &NiagaraMeshRendererProperties
    &NiagaraSpriteRendererProperties
    &NiagaraLightRendererProperties
   ]
  }
  {emitterName=Flames_Packed, rendererClasses [&NiagaraMeshRendererProperties, &NiagaraSpriteRendererProperties]}
  {emitterName=Embers_Packed, rendererClasses [&NiagaraSpriteRendererProperties]}
 ]
 userVariables [
  User.bIsFrontEnd {type=NiagaraBool}
  User.bIsFrontEndPreview {type=NiagaraBool}
  User.bisSwinging {type=NiagaraBool}
 ]
}
